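# Physics asset for the Sweet Teriyaki Green backpack cosmetic, covering a creature rig with body, head, jaw, tail and four fin chains (top, bottom, inner, outer).
# Each dyn_ bone gets one capsule body.
# Constraints form short chains rooted at dyn_body: the jaw hangs off the head, and tail and fins have two segments each.
# Tight twist limits and uneven swing limits keep each joint moving mostly along one axis.
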PhysicsAsset /BRCosmetics/Accessories/FORT_Backpacks/Backpack_M_MED_Sweet_Teriyaki_Green/Meshes/M_MED_Sweet_Teriyaki_Green_Pack_Physics
bodies {
 dyn_body [{shapeType=Capsule, length=4.3495, radius=2.1242, shapeName=dyn_body_capsule, rotation=<rot -90 -0.0395 0.0394>}]
 dyn_fin_bot_01 [{shapeType=Capsule, length=5.3587, radius=2.9619, shapeName=dyn_fin_bot_01_capsule, center=<0.5 0 0>}]
 dyn_fin_bot_02 [{shapeType=Capsule, length=3.3383, radius=2.7376, shapeName=dyn_fin_bot_02_capsule, center=<0.5 0 0>}]
 dyn_fin_in_01 [{shapeType=Capsule, length=2.25, radius=1, shapeName=dyn_fin_in_01_capsule, center=<0.5 0 0>, rotation=<rot -90 0 0>}]
 dyn_fin_in_02 [{shapeType=Capsule, length=2, radius=1, shapeName=dyn_fin_in_02_capsule, center=<0.5 0 0>, rotation=<rot -90 0 0>}]
 dyn_fin_out_01 [{shapeType=Capsule, length=2, radius=1, shapeName=dyn_fin_out_01_capsule, center=<-0.5 0 0>, rotation=<rot 90 0 0>}]
 dyn_fin_out_02 [{shapeType=Capsule, length=2, radius=1, shapeName=dyn_fin_out_02_capsule, center=<-0.5 0 0>, rotation=<rot 90 0 0>}]
 dyn_fin_top_01 [{shapeType=Capsule, length=7.4723, radius=1.4648, shapeName=dyn_fin_top_01_capsule}]
 dyn_fin_top_02 [{shapeType=Capsule, length=5.1747, radius=1.3008, shapeName=dyn_fin_top_02_capsule}]
 dyn_head [{shapeType=Capsule, length=3.8492, radius=2.4289, shapeName=dyn_head_capsule, center=<0.5 0.0001 0>}]
 dyn_jaw [{shapeType=Capsule, length=3.6678, radius=1.0284, shapeName=dyn_jaw_capsule, center=<1 0 0>, rotation=<rot 90 -0.0003 -0.0002>}]
 dyn_tail_01 [{shapeType=Capsule, length=2.6912, radius=1.1925, shapeName=dyn_tail_01_capsule, center=<1 0 0>}]
 dyn_tail_02 [{shapeType=Capsule, length=2.005, radius=1.255, shapeName=dyn_tail_02_capsule, rotation=<rot -0.1404 -90 -4.4872>}]
}
constraints [
 {bone1Name=dyn_tail_01, bone2Name=dyn_body, swing1LimitDegrees=25, swing2LimitDegrees=1, twistLimitDegrees=1}
 {bone1Name=dyn_tail_02, bone2Name=dyn_tail_01, swing1LimitDegrees=25, swing2LimitDegrees=1, twistLimitDegrees=1}
 {bone1Name=dyn_fin_top_01, bone2Name=dyn_body, swing1LimitDegrees=10, swing2LimitDegrees=2, twistLimitDegrees=1}
 {bone1Name=dyn_fin_top_02, bone2Name=dyn_fin_top_01, swing1LimitDegrees=15, swing2LimitDegrees=5, twistLimitDegrees=1}
 {bone1Name=dyn_fin_bot_01, bone2Name=dyn_body, swing1LimitDegrees=25, swing2LimitDegrees=5, twistLimitDegrees=1}
 {bone1Name=dyn_fin_bot_02, bone2Name=dyn_fin_bot_01, swing1LimitDegrees=25, swing2LimitDegrees=5, twistLimitDegrees=1}
 {bone1Name=dyn_head, bone2Name=dyn_body, swing1LimitDegrees=25, swing2LimitDegrees=5, twistLimitDegrees=1}
 {bone1Name=dyn_jaw, bone2Name=dyn_head, swing1LimitDegrees=5, swing2LimitDegrees=15, twistLimitDegrees=1}
 {bone1Name=dyn_fin_in_01, bone2Name=dyn_body, swing1LimitDegrees=5, swing2LimitDegrees=25, twistLimitDegrees=1}
 {bone1Name=dyn_fin_in_02, bone2Name=dyn_fin_in_01, swing1LimitDegrees=5, swing2LimitDegrees=25, twistLimitDegrees=1}
 {bone1Name=dyn_fin_out_01, bone2Name=dyn_body, swing1LimitDegrees=5, swing2LimitDegrees=25, twistLimitDegrees=1}
 {bone1Name=dyn_fin_out_02, bone2Name=dyn_fin_out_01, swing1LimitDegrees=5, swing2LimitDegrees=25, twistLimitDegrees=1}
]
